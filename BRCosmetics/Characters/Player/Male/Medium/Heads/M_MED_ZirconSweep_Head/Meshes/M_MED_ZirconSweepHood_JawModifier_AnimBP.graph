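# Animation Blueprint for the ZirconSweep Hood male head on the Fortnite base head skeleton, repositioning mouth and jaw bones.
# It converts the input pose to component space and chains Transform (Modify) Bone nodes through upper mouth, upper teeth, lower lip rotation and jaw.
# Each replaces translation with a fixed parent-space offset at full alpha, overriding any incoming translation on those bones.
# It has no variables or active event logic.
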AnimBlueprint /BRCosmetics/Characters/Player/Male/Medium/Heads/M_MED_ZirconSweep_Head/Meshes/M_MED_ZirconSweepHood_JawModifier_AnimBP
target_skeleton=/Game/Characters/Player/Common/Fortnite_Base_Head/Fortnite_Base_3L_Head_Skeleton
graphs [
 AnimGraph {nodes [
   n1 = "Transform (Modify) Bone - Bone: FACIAL_C_MouthUpper"(ComponentPose=Local To Component(LocalPose=Input Pose()), Alpha=1, Translation="0.000,-4.775,-0.000", Rotation="0,0,0", Scale="1,1,1") {
    translationMode=BMM_Replace
    translationSpace=BCS_ParentBoneSpace
    alphaBoolBlend {blendOption=Linear}
    boneToModify {boneName=FACIAL_C_MouthUpper}
    translation {y=-4.7754, z=0}
   }
   n2 = "Transform (Modify) Bone - Bone: FACIAL_C_TeethUpper"(ComponentPose=@n1, Alpha=1, Translation="-0.000,4.423,5.214", Rotation="0,0,0", Scale="1,1,1") {
    translationMode=BMM_Replace
    translationSpace=BCS_ParentBoneSpace
    alphaBoolBlend {blendOption=Linear}
    boneToModify {boneName=FACIAL_C_TeethUpper}
    translation {y=4.423, z=5.214}
   }
   n3 = "Transform (Modify) Bone - Bone: FACIAL_C_LowerLipRotation"(ComponentPose=@n2, Alpha=1, Translation="-0.000,1.842,-3.000", Rotation="0,0,0", Scale="1,1,1") {
    translationMode=BMM_Replace
    translationSpace=BCS_ParentBoneSpace
    alphaBoolBlend {blendOption=Linear}
    boneToModify {boneName=FACIAL_C_LowerLipRotation}
    translation {y=1.842, z=-3}
   }
   n4 = "Transform (Modify) Bone - Bone: FACIAL_C_Jaw"(ComponentPose=@n3, Alpha=1, Translation="0.000,1.438,-3.760", Rotation="0,0,0", Scale="1,1,1") {
    scaleMode=BMM_Additive
    translationMode=BMM_Replace
    translationSpace=BCS_ParentBoneSpace
    alphaBoolBlend {blendOption=Linear}
    boneToModify {boneName=FACIAL_C_Jaw}
    translation {y=1.438, z=-3.76}
   }
   Output Pose(Result=Component To Local(ComponentPose=@n4)) {layerGroup=None}
  ]}
 EventGraph {nodes [
   TryGetPawnOwner()
   Event BlueprintUpdateAnimation()
  ]}
]
variables []
